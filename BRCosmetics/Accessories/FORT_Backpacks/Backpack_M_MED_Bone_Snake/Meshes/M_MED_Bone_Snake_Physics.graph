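# Physics asset for the Bone Snake backpack cosmetic, giving its sword and dangling cloth bones physical secondary motion.
# A box body on the pack anchors the sword, which has narrow swing and twist limits.
# Left and right five-bone cloth chains hang from the sword with wider swing limits.
# A tighter, asymmetrically limited third chain hangs from a master bone whose rotation relative to the sword is free.
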PhysicsAsset /BRCosmetics/Accessories/FORT_Backpacks/Backpack_M_MED_Bone_Snake/Meshes/M_MED_Bone_Snake_Physics
bodies {
 dyn_cloth_l_01 [{shapeType=Box, extentX=2.9395, extentY=1.1, extentZ=2.1347, center=<0.9014 0 0>, rotation=<rot 90 0 0>}]
 dyn_cloth_l_02 [{shapeType=Box, extentX=2.9395, extentY=1.1, extentZ=2.4072, center=<1.4187 0 0>, rotation=<rot 90 0 0>}]
 dyn_cloth_l_03 [{shapeType=Box, extentX=2.9395, extentY=1.1, extentZ=4.2607, center=<2.3011 0 0>, rotation=<rot 90 0 0>}]
 dyn_cloth_l_04 [{shapeType=Box, extentX=2.9395, extentY=1.1, extentZ=5.1171, center=<2.4954 0 0>, rotation=<rot 90 0 0>}]
 dyn_cloth_l_05 [{shapeType=Box, extentX=2.9395, extentY=1.1, extentZ=7.3148, center=<3.4574 0 0>, rotation=<rot 90 0 0>}]
 dyn_cloth_r_01 [{shapeType=Box, extentX=2.9395, extentY=1.1, extentZ=2.3444, center=<1.0389 0 0>, rotation=<rot 90 0 0>}]
 dyn_cloth_r_02 [{shapeType=Box, extentX=2.9395, extentY=1.1, extentZ=2.2877, center=<1.0422 0 0>, rotation=<rot 90 0 0>}]
 dyn_cloth_r_03 [{shapeType=Box, extentX=2.9395, extentY=1.1, extentZ=4, center=<2 0 0>, rotation=<rot 90 0 0>}]
 dyn_cloth_r_04 [{shapeType=Box, extentX=2.9395, extentY=1.1, extentZ=6.1835, center=<3.1806 0 0>, rotation=<rot 90 0 0>}]
 dyn_cloth_r_05 [{shapeType=Box, extentX=2.9395, extentY=1.1, extentZ=8.0628, center=<3.7276 0 0>, rotation=<rot 90 0 0>}]
 dyn_cloth_t_01 [{shapeType=Box, extentX=1.1, extentY=1.1, extentZ=1.5223, center=<0.8469 0 0>, rotation=<rot 90 0 0>}]
 dyn_cloth_t_02 [{shapeType=Box, extentX=1.1, extentY=1.1, extentZ=2.0932, center=<1.0605 0 0>, rotation=<rot 90 0 0>}]
 dyn_cloth_t_03 [{shapeType=Box, extentX=1.1, extentY=1.1, extentZ=2.2588, center=<0.889 0 0>, rotation=<rot 90 0 0>}]
 dyn_cloth_t_04 [{shapeType=Box, extentX=1.1, extentY=1.1, extentZ=4, center=<2 0 0>, rotation=<rot 90 0 0>}]
 dyn_cloth_t_mstr [{shapeType=Box, extentX=1.0979, extentY=1.2917, extentZ=0.8963, center=<-0.0782 0 0>, rotation=<rot 90 0 0>}]
 dyn_sword [{shapeType=Box, extentX=7.5966, extentY=2.4713, extentZ=85.9815, center=<-5.7664 -2.566 -7.4423>, rotation=<rot -30.8596 0 0>}]
 pack [{shapeType=Box, extentX=18.3663, extentY=36.197, extentZ=42.0769, center=<-1.8382 -4.6287 -9.4692>, rotation=<rot 0 -90 0>}]
}
constraints [
 {bone1Name=dyn_sword, bone2Name=pack, swing1LimitDegrees=15, swing2LimitDegrees=15, twistLimitDegrees=7.5}
 {bone1Name=dyn_cloth_r_01, bone2Name=dyn_sword, swing1LimitDegrees=45, swing2LimitDegrees=45, twistLimitDegrees=7.5}
 {bone1Name=dyn_cloth_r_02, bone2Name=dyn_cloth_r_01, swing1LimitDegrees=45, swing2LimitDegrees=45, twistLimitDegrees=7.5}
 {bone1Name=dyn_cloth_r_03, bone2Name=dyn_cloth_r_02, swing1LimitDegrees=45, swing2LimitDegrees=45, twistLimitDegrees=7.5}
 {bone1Name=dyn_cloth_r_04, bone2Name=dyn_cloth_r_03, swing1LimitDegrees=45, swing2LimitDegrees=45, twistLimitDegrees=7.5}
 {bone1Name=dyn_cloth_r_05, bone2Name=dyn_cloth_r_04, swing1LimitDegrees=45, swing2LimitDegrees=45, twistLimitDegrees=7.5}
 {bone1Name=dyn_cloth_l_01, bone2Name=dyn_sword, swing1LimitDegrees=45, swing2LimitDegrees=45, twistLimitDegrees=5}
 {bone1Name=dyn_cloth_l_02, bone2Name=dyn_cloth_l_01, swing1LimitDegrees=45, swing2LimitDegrees=45, twistLimitDegrees=5}
 {bone1Name=dyn_cloth_l_03, bone2Name=dyn_cloth_l_02, swing1LimitDegrees=45, swing2LimitDegrees=45, twistLimitDegrees=5}
 {bone1Name=dyn_cloth_l_04, bone2Name=dyn_cloth_l_03, swing1LimitDegrees=45, swing2LimitDegrees=45, twistLimitDegrees=5}
 {bone1Name=dyn_cloth_l_05, bone2Name=dyn_cloth_l_04, swing1LimitDegrees=45, swing2LimitDegrees=45, twistLimitDegrees=5}
 {bone1Name=dyn_cloth_t_01, bone2Name=dyn_cloth_t_mstr, swing1LimitDegrees=25, swing2LimitDegrees=5, twistLimitDegrees=10}
 {bone1Name=dyn_cloth_t_02, bone2Name=dyn_cloth_t_01, swing1LimitDegrees=25, swing2LimitDegrees=5, twistLimitDegrees=10}
 {bone1Name=dyn_cloth_t_03, bone2Name=dyn_cloth_t_02, swing1LimitDegrees=25, swing2LimitDegrees=5, twistLimitDegrees=10}
 {bone1Name=dyn_cloth_t_04, bone2Name=dyn_cloth_t_03, swing1LimitDegrees=25, swing2LimitDegrees=5, twistLimitDegrees=10}
 {
  bone1Name=dyn_cloth_t_mstr
  bone2Name=dyn_sword
  swing1LimitDegrees=45
  swing1Motion=Free
  swing2LimitDegrees=45
  swing2Motion=Free
  twistLimitDegrees=45
  twistMotion=Free
 }
]
